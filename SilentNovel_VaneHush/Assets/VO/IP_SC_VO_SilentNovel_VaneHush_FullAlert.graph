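# Sound Cue for Silent Novel (Vane Hush) full-alert voice lines, set up as NPC boss combat chatter.
# A single Random node picks among eleven equally weighted wave players, each holding one FullAlert VO take.
# The cue uses the shared henchmen attenuation, the critical enemy VO sound class and boss combat chatter concurrency.
# It also sends to sidechain volume buses for music, first-person weapon, weapon and explosion audio.
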SoundCue /SilentNovel_VaneHush/Assets/VO/IP_SC_VO_SilentNovel_VaneHush_FullAlert
root=Random_1
nodes [
 {
  id=Random_1
  children [
   WavePlayer_0
   WavePlayer_1
   WavePlayer_2
   WavePlayer_3
   WavePlayer_4
   WavePlayer_5
   WavePlayer_6
   WavePlayer_7
   WavePlayer_8
   WavePlayer_9
   WavePlayer_10
  ]
  properties {weights [1,1,1,1,1,1,1,1,1,1,1]}
 }
 {id=WavePlayer_0, properties {soundWaveAssetPtr=/SilentNovel_VaneHush/Assets/VO/wav/vo_IP_WildEstate_SilentNovel_FullAlert_010}}
 {id=WavePlayer_1, properties {soundWaveAssetPtr=/SilentNovel_VaneHush/Assets/VO/wav/vo_IP_WildEstate_SilentNovel_FullAlert_020}}
 {id=WavePlayer_10, properties {soundWaveAssetPtr=/SilentNovel_VaneHush/Assets/VO/wav/vo_IP_WildEstate_SilentNovel_FullAlert_110}}
 {id=WavePlayer_2, properties {soundWaveAssetPtr=/SilentNovel_VaneHush/Assets/VO/wav/vo_IP_WildEstate_SilentNovel_FullAlert_030}}
 {id=WavePlayer_3, properties {soundWaveAssetPtr=/SilentNovel_VaneHush/Assets/VO/wav/vo_IP_WildEstate_SilentNovel_FullAlert_040}}
 {id=WavePlayer_4, properties {soundWaveAssetPtr=/SilentNovel_VaneHush/Assets/VO/wav/vo_IP_WildEstate_SilentNovel_FullAlert_050}}
 {id=WavePlayer_5, properties {soundWaveAssetPtr=/SilentNovel_VaneHush/Assets/VO/wav/vo_IP_WildEstate_SilentNovel_FullAlert_060}}
 {id=WavePlayer_6, properties {soundWaveAssetPtr=/SilentNovel_VaneHush/Assets/VO/wav/vo_IP_WildEstate_SilentNovel_FullAlert_070}}
 {id=WavePlayer_7, properties {soundWaveAssetPtr=/SilentNovel_VaneHush/Assets/VO/wav/vo_IP_WildEstate_SilentNovel_FullAlert_080}}
 {id=WavePlayer_8, properties {soundWaveAssetPtr=/SilentNovel_VaneHush/Assets/VO/wav/vo_IP_WildEstate_SilentNovel_FullAlert_090}}
 {id=WavePlayer_9, properties {soundWaveAssetPtr=/SilentNovel_VaneHush/Assets/VO/wav/vo_IP_WildEstate_SilentNovel_FullAlert_100}}
]
properties {
 attenuationSettings=/Game/Sounds/Fort_Human_VO_BR/NPCs/Shared/NPC_VO_Henchmen_Att
 maxDistance=8500
 soundClassObject=&Fort_VO_Enemy_Critical
 volumeMultiplier=1.8836
 busSends [
  {audioBus=&AB_Sidechain_FortMusic_Vol, minSendLevel=0, sendLevel=1, soundSourceBus=null}
  {audioBus=/Game/Sounds/Modulation/EnvelopeFollowers/Volume/AB_Sidechain_FortWeapon1P_Vol, minSendLevel=0, sendLevel=1, soundSourceBus=null}
  {audioBus=&AB_Sidechain_FortWeapon_Vol, minSendLevel=0, sendLevel=1, soundSourceBus=null}
  {audioBus=&AB_Sidechain_FortExplosion_Vol, minSendLevel=0, sendLevel=1, soundSourceBus=null}
 ]
 concurrencySet [&NPC_Boss_Combat_Chatter_Conc]
}
