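# Sound cue for the Vane Hush NPC's idle dancing voice in SilentNovel, playing a Phase 0 combat-music vocal loop wave.
# A single wave player feeds the output directly, with boosted volume, the enemy critical VO class and shared henchmen VO attenuation.
# It shares boss combat chatter concurrency and sends to music, weapon and first-person weapon sidechain buses, so the vocal can drive ducking of those mixes.
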SoundCue /SilentNovel_VaneHush/Assets/VO/IP_SC_VO_SilentNovel_VaneHush_Idle_Dancing_Voice
root=WavePlayer_2
nodes [{
  id=WavePlayer_2
  properties {soundWaveAssetPtr=/SilentNovel_VaneHush/Assets/Audio/SoundWaves/MX_S38_Combat_IP_SilentNovel_VaneHush_Phase_0_Vocal_LP}
 }]
properties {
 attenuationSettings=/Game/Sounds/Fort_Human_VO_BR/NPCs/Shared/NPC_VO_Henchmen_Att
 maxDistance=8500
 soundClassObject=&Fort_VO_Enemy_Critical
 volumeMultiplier=2
 busSends [
  {audioBus=&AB_Sidechain_FortMusic_Vol, minSendLevel=0, sendLevel=1, soundSourceBus=null}
  {audioBus=/Game/Sounds/Modulation/EnvelopeFollowers/Volume/AB_Sidechain_FortWeapon1P_Vol, minSendLevel=0, sendLevel=1, soundSourceBus=null}
  {audioBus=&AB_Sidechain_FortWeapon_Vol, minSendLevel=0, sendLevel=1, soundSourceBus=null}
 ]
 concurrencySet [&NPC_Boss_Combat_Chatter_Conc]
}
