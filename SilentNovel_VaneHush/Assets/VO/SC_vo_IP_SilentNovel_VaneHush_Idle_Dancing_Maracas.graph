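# Sound cue in the VO folder for the Vane Hush character, named for a maracas idle dance.
# A single wave player feeds the output directly, playing a Phase 0 combat music percussion wave rather than a voice line.
# It uses shared henchman VO attenuation, the critical enemy VO sound class and boss combat chatter concurrency, with boosted volume.
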SoundCue /SilentNovel_VaneHush/Assets/VO/SC_vo_IP_SilentNovel_VaneHush_Idle_Dancing_Maracas
root=WavePlayer_0
nodes [{
  id=WavePlayer_0
  properties {soundWaveAssetPtr=/SilentNovel_VaneHush/Assets/Audio/SoundWaves/MX_S38_Combat_IP_SilentNovel_VaneHush_Phase_0_Perc_LP}
 }]
properties {
 attenuationSettings=/Game/Sounds/Fort_Human_VO_BR/NPCs/Shared/NPC_VO_Henchmen_Att
 maxDistance=8500
 soundClassObject=&Fort_VO_Enemy_Critical
 volumeMultiplier=2
 concurrencySet [&NPC_Boss_Combat_Chatter_Conc]
}
